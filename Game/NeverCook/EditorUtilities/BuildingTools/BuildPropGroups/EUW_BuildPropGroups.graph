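# Editor utility wizard for building simple destructible prop groups on the supported Terrain maps.
# A widget switcher offers to load the first listed map unless the current map qualifies or build controls are forced.
# Load pins World Partition actors for prop groups, explosion proxies and FoliageCluster/CropRow Blueprints; Build runs the BuildGroups console command.
# A checkbox shows groups and proxies while hiding those foliage actors, or the reverse.
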EditorUtilityWidgetBlueprint /Game/NeverCook/EditorUtilities/BuildingTools/BuildPropGroups/EUW_BuildPropGroups
graphs [
 BuildSimplePropGroups {nodes [
   BuildSimplePropGroups() {is=title K2Node_FunctionEntry}
   > ExecuteConsoleCommandFromEditor!(execute, ConsoleCommand:String=SimpleDestructiblePropGroups.BuildGroups) -> (then)
  ]}
 DoTogglePropGroups {nodes [
   DoTogglePropGroups() {is=title K2Node_FunctionEntry}
   > n1 = Sequence()
   > TogglePropsVisibilityByClass!(execute, self, PropClass:Actor Class Reference=/Script/FortniteGame.SimpleDestructiblePropGroup) -> (then)
   > TogglePropsVisibilityByClass(PropClass=/Script/FortniteGame.SimpleDestructiblePropExplosionProxy)
   n2 = GetAllLevelActors(execute=@n1.then_1, self=Get EditorActorSubsystem())
   > n3 = For Each Loop(Array:Array of Actor Object References=@n2) -> (Array Element:Actor Object Reference)
   n4 = For Each Loop(Exec=@n3.Completed, Array=Get l_actors() -> (l_actors:Array of Actor Object References))
   > SetIsTemporarilyHiddenInEditor!(
    execute
    self:Actor Object Reference=@n4
    bIsHidden:bool=getSelected!(self:TA Base Editor Button Widget Object Reference=Get btnTogglePropVis() -> (btnTogglePropVis:TA Generic 01 Checkbox Object Reference)) -> (isSelected:bool)
   ) -> (then)
   n5 = Get Class(Object=@n3)
   n6 = IsValidClass!(
    execute=@n3.LoopBody
    self
    classString:String=BreakSoftObjectPath(InSoftObjectPath=GetSoftClassPath!(Class:Object Class Reference=@n5) -> (ReturnValue:Soft Class Path Structure))
   ) -> (then, valid:bool)
   > Branch(Condition=@n6)
   > Add(TargetArray:Array of Actor Object References=Get l_actors(), NewItem:"Actor Object Reference (by ref)"=@n3) {is=title K2Node_CallArrayFunction}
  ]}
 EventGraph {nodes [
   "On Clicked (btnLoad)"!() -> (OutputDelegate, then, widget:User Widget Object Reference, event:Pointer Event Structure) {is=title K2Node_ComponentBoundEvent}
   > LoadActorsWithInterface()
   "On Clicked (btnBuild)"!() -> (OutputDelegate, then, widget:User Widget Object Reference, event:Pointer Event Structure) {is=title K2Node_ComponentBoundEvent}
   > BuildSimplePropGroups() {is=title K2Node_CallFunction}
   "On Clicked (btnLoadMap)"!() -> (OutputDelegate, then, widget:User Widget Object Reference, event:Pointer Event Structure) {is=title K2Node_ComponentBoundEvent}
   > LoadMap!(
    execute
    Filename:String="Get (a ref)"(Array:Array of Strings=Get MapPaths() -> (MapPaths:Array of Strings), Dimension 1=0) -> (Output:"String (by ref)")
   ) -> (then, ReturnValue:World Object Reference)
   "On Clicked (btnTogglePropVis)"!() -> (OutputDelegate, then, widget:User Widget Object Reference, event:Pointer Event Structure) {is=title K2Node_ComponentBoundEvent}
   > DoTogglePropGroups() {is=title K2Node_CallFunction}
   "On Clicked (btnForceShow)"!() -> (OutputDelegate, then, widget:User Widget Object Reference, event:Pointer Event Structure) {is=title K2Node_ComponentBoundEvent}
   > n1 = Branch(Condition=getSelected(self=Get btnForceShow() -> (btnForceShow:TA Generic 01 Checkbox Object Reference)))
   > SetActiveWidgetIndex(self=Get WidgetSwitcher_218() -> (WidgetSwitcher_218:Widget Switcher Object Reference), Index=1)
   initStage(execute=@n1.else)
   Event Construct()
   > initStage()
  ]}
 GetMapName {nodes [
   GetMapName!() -> (then, InMapPath:String) {is=title K2Node_FunctionEntry}
   > n1 = Sequence!(execute) -> (then_0, then_1, then_2:Exec)
   n2 = Split(SourceString=Get InMapPath() -> (InMapPath:String), InStr=/, SearchDir=FromEnd)
   > Branch(Condition=@n2.ReturnValue)
   > Return Node!(execute, MapName:String=@n2.RightS)
   Return Node(execute=@n1.then_2)
   n3 = Split(SourceString=Get InMapPath(), InStr="\\", SearchDir=FromEnd)
   Branch(execute=@n1.then_1, Condition=@n3.ReturnValue)
   > Return Node(MapName=@n3.RightS)
  ]}
 IsOneOfValidMaps {nodes [
   n1 = IsOneOfValidMaps!() -> (then, InMapName:String) {is=title K2Node_FunctionEntry}
   > n2 = For Each Loop(Array:Array of Strings=Get MapPaths()) -> (Array Element:String)
   Return Node!(execute=@n2.Completed, IsValid:bool=false)
   > Branch(Condition="Equal Exactly (String)"(A=GetMapName!(InMapPath:String=@n2) -> (MapName:String) {is=title K2Node_CallFunction}, B=@n1))
   > Return Node(IsValid=true)
  ]}
 IsValidClass {nodes [
   n1 = IsValidClass!() -> (then, classString:String) {is=title K2Node_FunctionEntry}
   > n2 = Sequence!(execute) -> (then_0, then_1)
   > Branch(Condition=NOT Boolean(A=StartsWith(SourceString=@n1, InPrefix=/Script)))
   n3 = Split(SourceString=@n1, InStr=., SearchDir=FromEnd)
   > Branch(Condition=@n3.ReturnValue)
   > n4 = Set l_classString(l_classString:String=@n3.RightS) -> (Output_Get:String)
   > n5 = To Array!(execute, A:Set of Strings=Get ValidClassStrings() -> (ValidClassStrings:Set of Strings)) -> (then, Result:Array of Strings)
   > n6 = For Each Loop with Break!(Exec:Exec, Array:Array of Strings=@n5, Break:Exec) -> (LoopBody:Exec, Array Element:String, Array Index:int, Completed:Exec)
   > Branch(Condition=Contains(SearchIn=@n4, Substring=@n6))
   > Return Node!(execute, valid:bool=true)
   Return Node(execute=@n2.then_1, valid=false)
  ]}
 LoadActorsWithInterface {nodes [
   LoadActorsWithInterface() {is=title K2Node_FunctionEntry}
   > n1 = Sequence()
   > n2 = GetActorDescs!(execute) -> (then, OutActorDescs:Array of Actor Desc Structures, ReturnValue:bool)
   > n3 = For Each Loop(Array:Array of Actor Desc Structures=@n2) -> (Array Element:Actor Desc Structure)
   n4 = Break Actor Desc!(ActorDesc:"Actor Desc Structure (by ref)"=@n3) -> (
    Guid:Guid Structure
    NativeClass:Object Class Reference
    Class:Soft Object Path Structure
    Name:Name
    Label:Name
    Bounds:Box Structure
    RuntimeGrid:Name
    bIsSpatiallyLoaded:bool=~
    bActorIsEditorOnly:bool=~
    ActorPackage:Name
    ActorPath:Name
    FolderPath:Name
    DataLayerAssets:Array of Soft Object Path Structures
   ) {is=title K2Node_BreakStruct}
   > Set l_guid(l_guid:Guid Structure=@n4) -> (Output_Get:Guid Structure)
   > n5 = Sequence()
   > n6 = Cast To SimpleDestructiblePropGroup Class!(execute:Exec, Class:Object Class Reference=@n4.NativeClass) -> (then:Exec, CastFailed:Exec, AsSimple Destructible Prop Group:Simple Destructible Prop Group Class Reference) {is=title K2Node_ClassDynamicCast}
   > Add(
    TargetArray:Array of Guid Structures=Get l_actorsToPin() -> (l_actorsToPin:Array of Guid Structures)
    NewItem:"Guid Structure (by ref)"=Get l_guid() -> (l_guid:Guid Structure)
   )
   n7 = IsValidClass(execute=@n6.CastFailed, classString=BreakSoftObjectPath(InSoftObjectPath=@n4.Class)) {is=title K2Node_CallFunction}
   > Branch(Condition=@n7)
   > Add(TargetArray=Get l_actorsToPin(), NewItem=Get l_guid())
   Cast To SimpleDestructiblePropExplosionProxy Class!(execute:Exec=@n5.then_1, Class:Object Class Reference=@n4.NativeClass) -> (then:Exec, CastFailed:Exec, AsSimple Destructible Prop Explosion Proxy:Simple Destructible Prop Explosion Proxy Class Reference) {is=title K2Node_ClassDynamicCast}
   > Add(TargetArray=Get l_actorsToPin(), NewItem=Get l_guid())
   PinActors!(execute=@n1.then_1, InActorsToPin:Array of Guid Structures=Get l_actorsToPin()) -> (then)
  ]}
 TogglePropActorVisibility {nodes [
   n1 = TogglePropActorVisibility!() -> (then, PropActor:Actor Object Reference) {is=title K2Node_FunctionEntry}
   n2 = getSelected(self=Get btnTogglePropVis())
   > SetEditorProperty(Object=@n1, PropertyName=bAllowVisibility, PropertyValue=@n2)
   > SetIsTemporarilyHiddenInEditor(self=@n1, bIsHidden=NOT Boolean(A=@n2))
  ]}
 TogglePropsVisibilityByClass {nodes [
   n1 = TogglePropsVisibilityByClass!() -> (then, PropClass:Actor Class Reference) {is=title K2Node_FunctionEntry}
   > n2 = GetAllLevelActors(self=Get EditorActorSubsystem())
   > n3 = Filter by Class!(execute, TargetArray:Array of Object References=@n2, ObjectClass:Object Class Reference=@n1, FilterType:EEditorScriptingFilterType Enum) -> (then, ReturnValue:Array of Actor Object References)
   > n4 = For Each Loop(Array:Array of Actor Object References=@n3) -> (Array Element:Actor Object Reference)
   > TogglePropActorVisibility!(execute, self, PropActor:Actor Object Reference=@n4) -> (then) {is=title K2Node_CallFunction}
  ]}
 initStage {nodes [
   initStage() {is=title K2Node_FunctionEntry}
   > n1 = Sequence()
   n2 = "Get (a ref)"(Array=Get MapPaths(), Dimension 1=0)
   > setLabel!(
    execute
    self:TA Base Editor Button Widget Object Reference=Get btnLoadMap() -> (btnLoadMap:TA Generic 01 Button Object Reference)
    label:String=Append(A="Load ", B=GetMapName(InMapPath=@n2))
   ) -> (then)
   n3 = GetCurrentLevelName!(execute=@n1.then_1, WorldContextObject:Object Reference, bRemovePrefixString:bool=true) -> (then, ReturnValue:String)
   > n4 = Branch(Condition=IsOneOfValidMaps!(InMapName:String=@n3) -> (IsValid:bool) {is=title K2Node_CallFunction})
   > SetActiveWidgetIndex(self=Get WidgetSwitcher_218(), Index=1)
   SetActiveWidgetIndex(execute=@n4.else, self=Get WidgetSwitcher_218(), Index=0)
  ]}
]
variables [
 MapPaths {
  type="array<string>"
  default [/Game/Athena/Helios/Maps/Helios_Terrain, /Game/Athena/Asteria/Maps/Asteria_Terrain, /BRMapCh6/Maps/Hermes_Terrain]
 }
 ValidClassStrings {type="array<string>", default [FoliageCluster, CropRow]}
]
widgets {
 info {rootWidgetClass=&StackBox, widgetCount=19}
 widgets [
  StackBox_0 {widgetClassPath=&StackBox, widget=StackBox_0, properties {orientation=Orient_Vertical}}
  ScaleBox_0 {
   widgetClassPath=&ScaleBox
   parent=StackBox_0
   widget=ScaleBox_0
   properties {stretch=UserSpecified, userSpecifiedScale=0.5}
   slot {class=StackBoxSlot, properties {horizontalAlignment=HAlign_Left, verticalAlignment=VAlign_Bottom, padding=<ltrb 5 5 5 5>}}
  }
  btnForceShow {
   widgetClassPath=/Game/Blueprints/Blutilities/ForTechArt/Widgets/Buttons/TA_Generic01_Checkbox
   bIsVariable=true
   parent=ScaleBox_0
   widget=btnForceShow
   properties {
    label=Force Show Build Controls
    toolTipText=Normally this acts as a wizard to direct people to the proper map but if you want to run in your map click here.
   }
   slot {class=ScaleBoxSlot}
  }
  WidgetSwitcher_218 {widgetClassPath=&WidgetSwitcher, bIsVariable=true, parent=StackBox_0, widget=WidgetSwitcher_218, slot {class=StackBoxSlot}}
  VerticalBox_1 {widgetClassPath=&VerticalBox, parent=WidgetSwitcher_218, widget=VerticalBox_1, slot {class=WidgetSwitcherSlot}}
  txtMapNameInstructions {
   widgetClassPath=&CommonTextBlock
   bIsVariable=true
   parent=VerticalBox_1
   widget=txtMapNameInstructions
   properties {
    style=&TextStyle-Base_C
    text="Must be in one of the {*_Terrain} maps to build Prop Groups (e.g., Asteria_Terrain)"
    font {
     bForceMonospaced=false
     bMaterialIsStencil=false
     fontMaterial=null
     fontObject=&NotoSans
     letterSpacing=0
     monospacedWidth=1
     size=24
     skewAmount=0
     typefaceFontName=NotoSans-Regular
     outlineSettings {
      bApplyOutlineToDropShadows=false
      bMiteredCorners=false
      bSeparateFillAlpha=false
      outlineMaterial=null
      outlineSize=0
      outlineColor=<rgba 0 0 0 1>
     }
    }
    shadowOffset=<0 0>
   }
   slot {class=VerticalBoxSlot, properties {horizontalAlignment=HAlign_Center}}
  }
  btnLoadMap {
   widgetClassPath=&TA_Generic01_Button
   parent=VerticalBox_1
   widget=btnLoadMap
   properties {label=Load Helios Terrain}
   slot {class=VerticalBoxSlot}
  }
  VerticalBox_0 {
   widgetClassPath=&VerticalBox
   parent=WidgetSwitcher_218
   widget=VerticalBox_0
   slot {class=WidgetSwitcherSlot, properties {padding=<ltrb 10 10 10 10>}}
  }
  Border_0 {
   widgetClassPath=&Border
   parent=VerticalBox_0
   widget=Border_0
   properties {brushColor=<rgba 0.026 0.026 0.026 1>, padding=<ltrb 5 5 5 5>}
   slot {class=VerticalBoxSlot, properties {padding=<ltrb 10 10 10 10>}}
  }
  VerticalBox_2 {widgetClassPath=&VerticalBox, parent=Border_0, widget=VerticalBox_2, slot {class=BorderSlot, properties {padding=<ltrb 5 5 5 5>}}}
  CommonTextBlock {
   widgetClassPath=&CommonTextBlock
   parent=VerticalBox_2
   widget=CommonTextBlock
   properties {
    autoWrapText=true
    justification=Center
    text=First you have to load all foliage and prop groups
    font {
     bForceMonospaced=false
     bMaterialIsStencil=false
     fontMaterial=null
     fontObject=&NotoSans
     letterSpacing=0
     monospacedWidth=1
     size=18
     skewAmount=0
     typefaceFontName=NotoSans-Regular
     outlineSettings {
      bApplyOutlineToDropShadows=false
      bMiteredCorners=false
      bSeparateFillAlpha=false
      outlineMaterial=null
      outlineSize=0
      outlineColor=<rgba 0 0 0 1>
     }
    }
    shadowOffset=<0 0>
   }
   slot {class=VerticalBoxSlot}
  }
  btnLoad {
   widgetClassPath=&TA_Generic01_Button
   parent=VerticalBox_2
   widget=btnLoad
   properties {label=Step1: Load Necessary Actors}
   slot {class=VerticalBoxSlot, properties {padding=<ltrb 5 5 5 5>}}
  }
  Border_1 {
   widgetClassPath=&Border
   parent=VerticalBox_0
   widget=Border_1
   properties {brushColor=<rgba 0.026 0.026 0.026 1>, padding=<ltrb 5 5 5 5>}
   slot {class=VerticalBoxSlot, properties {padding=<ltrb 10 10 10 10>}}
  }
  VerticalBox_3 {widgetClassPath=&VerticalBox, parent=Border_1, widget=VerticalBox_3, slot {class=BorderSlot, properties {padding=<ltrb 5 5 5 5>}}}
  CommonTextBlock_1 {
   widgetClassPath=&CommonTextBlock
   parent=VerticalBox_3
   widget=CommonTextBlock_1
   properties {
    autoWrapText=true
    justification=Center
    text="After you have loaded all foliage and prop groups you can run the build command (it's very fast)."
    font {
     bForceMonospaced=false
     bMaterialIsStencil=false
     fontMaterial=null
     fontObject=&NotoSans
     letterSpacing=0
     monospacedWidth=1
     size=18
     skewAmount=0
     typefaceFontName=NotoSans-Regular
     outlineSettings {
      bApplyOutlineToDropShadows=false
      bMiteredCorners=false
      bSeparateFillAlpha=false
      outlineMaterial=null
      outlineSize=0
      outlineColor=<rgba 0 0 0 1>
     }
    }
    shadowOffset=<0 0>
   }
   slot {class=VerticalBoxSlot}
  }
  btnBuild {
   widgetClassPath=&TA_Generic01_Button
   parent=VerticalBox_3
   widget=btnBuild
   properties {label=Step 2: Build Prop Groups}
   slot {class=VerticalBoxSlot, properties {padding=<ltrb 5 5 5 5>}}
  }
  Border_2 {
   widgetClassPath=&Border
   parent=VerticalBox_0
   widget=Border_2
   properties {brushColor=<rgba 0.026 0.026 0.026 1>, padding=<ltrb 5 5 5 5>}
   slot {class=VerticalBoxSlot, properties {padding=<ltrb 10 10 10 10>}}
  }
  CommonTextBlock_2 {
   widgetClassPath=&CommonTextBlock
   parent=Border_2
   widget=CommonTextBlock_2
   properties {
    autoWrapText=true
    justification=Center
    text=After running the build command you need to CTRL+S and save any changed/added SimpleDestructiblePropGroups.
    font {
     bForceMonospaced=false
     bMaterialIsStencil=false
     fontMaterial=null
     fontObject=&NotoSans
     letterSpacing=0
     monospacedWidth=1
     size=18
     skewAmount=0
     typefaceFontName=NotoSans-Regular
     outlineSettings {
      bApplyOutlineToDropShadows=false
      bMiteredCorners=false
      bSeparateFillAlpha=false
      outlineMaterial=null
      outlineSize=0
      outlineColor=<rgba 0 0 0 1>
     }
    }
    shadowOffset=<0 0>
   }
   slot {class=BorderSlot, properties {padding=<ltrb 5 5 5 5>}}
  }
  btnTogglePropVis {
   widgetClassPath=&TA_Generic01_Checkbox
   bIsVariable=true
   parent=StackBox_0
   widget=btnTogglePropVis
   properties {label=Show Prop Groups In Editor}
   slot {
    class=StackBoxSlot
    properties {horizontalAlignment=HAlign_Center, verticalAlignment=VAlign_Bottom, padding=<ltrb 5 5 5 5>, size {sizeRule=Fill, value=1}}
   }
  }
 ]
}
